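# Niagara effect for the happy cat variant of the MincePounce player emote, framing it with start and end bursts.
# Spawn_Burst layers glows, flare, discs, sparkles and glitch meshes, with spawn groups driving per-layer settings and renderer visibility tags.
# Age curves animate size, color and dynamic parameters, and camera distance scales glitch size.
# End_Burst fires delayed purple sprites slowed by drag, while Cat and Light show no modules or renderers.
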NiagaraSystem /BRCosmetics/Animation/Game/MainPlayer/Emotes/MincePounce/Fx/NS_Emote_MincePounce_Cat_Happy
dependencies {
 usedDataInterfaces [&NiagaraDataInterfaceCurve, &NiagaraDataInterfaceColorCurve]
 usedDynamicInputs []
 usedModules [&SystemState]
}
emitters {
 Cat {rendererClasses [], renderers []}
 End_Burst {
  emitterUpdateScript [
   &EmitterState(Inactive Response=::NewEnumerator1, Life Cycle Mode=::NewEnumerator1, Loop Behavior=::NewEnumerator1, Loop Duration=3, UseLoopDelay=false)
   &SpawnBurst_Instantaneous(Spawn Group=1, Spawn Time=0.73, Use Spawn Probability=false, UseLoopCountLimit=false) {module=SpawnBurst_Instantaneous001}
   &SpawnBurst_Instantaneous(
    Spawn Count=&RandomRangeInt(Minimum=3, Maximum=4)
    Spawn Time=0.75
    Use Spawn Probability=false
    UseLoopCountLimit=false
   )
  ]
  particleSpawnScript [
   &InitializeParticle(
    Color Maximum=<rgba 0.733 0.3033 1 1>
    Color Minimum=<rgba 0.2692 0.0715 0.8125 1>
    Color Mode=::NewEnumerator2
    Lifetime=&MergeHelper_FloatSwitch(
     0=&RandomRangeFloat(Minimum=0.16, Maximum=0.35)
     1=0.18
     2=?
     3=?
     Index=&SpawnGroup()
     Num=::NewEnumerator0
    )
    Sprite Rotation Mode=::NewEnumerator2
    Sprite Size Mode=::NewEnumerator1
    Uniform Sprite Size Max=40
    Uniform Sprite Size Min=30
    UsePositionOffset=false
    Write Material Random=true
   )
   &AddVelocity(
    Module.Use Falloff Distance=false
    Override Seed=false
    Velocity Mode=::NewEnumerator1
    Velocity Speed=&MergeHelper_FloatSwitch(
     0=&RandomRangeFloat(Minimum=160, Maximum=260)
     1=0
     2=?
     3=?
     Index=&SpawnGroup()
     Num=::NewEnumerator0
    )
   )
   /SetVariables(Particles.Soft Sparkles Size:Vector2f=<130 90>, Particles.VisibilityTag=&SpawnGroup()) {isSetParameters=true}
  ]
  particleUpdateScript [
   &ParticleState()
   &ScaleSpriteSize(Uniform Curve Sprite Scale=DataInterface() {propertyValues {Curve {keys [
        <at 0 0 RCIM_Cubic>
        {arriveTangentWeight=0.0688, interpMode=RCIM_Cubic, leaveTangentWeight=0.3575, tangentMode=RCTM_User, time=0.1287, value=1}
        {arriveTangentWeight=0.3391, interpMode=RCIM_Cubic, leaveTangentWeight=0.103, tangentMode=RCTM_User, time=0.797, value=1}
        <at 1 0 RCIM_Cubic>
       ]}}})
   &Drag(
    Drag=&FloatFromCurve(
     FloatCurve=DataInterface() {propertyValues {Curve {keys [
         {arriveTangent=3.1214, interpMode=RCIM_Cubic, leaveTangent=3.1214, tangentMode=RCTM_User, time=0, value=0}
         {arriveTangent=1.6667, interpMode=RCIM_Cubic, leaveTangent=1.6667, time=0.2387, value=0.7}
         <at 0.6 1 RCIM_Cubic>
        ]}}}
     Scale Curve=&RandomRangeFloat(Minimum=15, Maximum=20)
    )
    Use Linear Drag=true
    Use Rotational Drag=true
   )
   &SolveForcesAndVelocity(Clamp Velocity=false, Limit Acceleration=false)
   &ScaleColor(
    Linear Color Curve=DataInterface() {propertyValues {AlphaCurve {keys [<at 0 0>,<at 0.0745 1>,<at 1 0>]}, BlueCurve {keys [<at 0 1>]}, GreenCurve {keys [<at 0 1>]}, RedCurve {keys [<at 0 1>]}}}
    Scale Mode=2
   )
   /SetVariables(Particles.Scale Soft Sparkles Size:Vector2f=&Vector2DFromFloat(Value=&Multiply_Float(
      A=&MakeFloatFromVector2d(Vector 2D=$Particles.Initial.Soft Sparkles Size)
      B=&FloatFromCurve(FloatCurve=DataInterface() {propertyValues {Curve {keys [<at 0 0 RCIM_Cubic>,<at 0.05 1 RCIM_Cubic>,<at 1 0 RCIM_Cubic>]}}})
     ))) {isSetParameters=true}
  ]
  rendererClasses [&NiagaraSpriteRendererProperties]
  renderers [
   {rendererClass=&NiagaraSpriteRendererProperties}
   {rendererClass=&NiagaraSpriteRendererProperties, rendererIndex=1}
   {rendererClass=&NiagaraSpriteRendererProperties, rendererIndex=2}
  ]
 }
 Light {rendererClasses [], renderers []}
 Spawn_Burst {
  emitterUpdateScript [
   &EmitterState(Life Cycle Mode=::NewEnumerator1, Loop Behavior=::NewEnumerator1, Loop Duration=1, UseLoopDelay=false)
   &SpawnBurst_Instantaneous(Spawn Count=10, Spawn Group=8, Use Spawn Probability=false, UseLoopCountLimit=false) {module=SpawnBurst_Instantaneous008}
   &SpawnBurst_Instantaneous(Spawn Count=10, Spawn Group=8, Spawn Time=0.15, Use Spawn Probability=false, UseLoopCountLimit=false) {module=SpawnBurst_Instantaneous012}
   &SpawnBurst_Instantaneous(Spawn Count=3, Spawn Group=9, Use Spawn Probability=false, UseLoopCountLimit=false) {module=SpawnBurst_Instantaneous011}
   &SpawnBurst_Instantaneous(Spawn Count=3, Spawn Group=9, Spawn Time=0.05, Use Spawn Probability=false, UseLoopCountLimit=false) {module=SpawnBurst_Instantaneous010}
   &SpawnBurst_Instantaneous(Spawn Count=10, Spawn Group=9, Spawn Time=0.1, Use Spawn Probability=false, UseLoopCountLimit=false) {module=SpawnBurst_Instantaneous009}
   &SpawnBurst_Instantaneous(Use Spawn Probability=false, UseLoopCountLimit=false) {module=SpawnBurst_Instantaneous001}
   &SpawnBurst_Instantaneous(Spawn Group=1, Use Spawn Probability=false, UseLoopCountLimit=false)
   &SpawnBurst_Instantaneous(Spawn Group=2, Use Spawn Probability=false, UseLoopCountLimit=false) {module=SpawnBurst_Instantaneous002}
   &SpawnBurst_Instantaneous(Spawn Group=3, Use Spawn Probability=false, UseLoopCountLimit=false) {module=SpawnBurst_Instantaneous003}
   &SpawnBurst_Instantaneous(Spawn Group=4, Use Spawn Probability=false, UseLoopCountLimit=false) {module=SpawnBurst_Instantaneous004}
   &SpawnBurst_Instantaneous(Spawn Group=5, Use Spawn Probability=false, UseLoopCountLimit=false) {module=SpawnBurst_Instantaneous005}
   &SpawnBurst_Instantaneous(
    Spawn Count=&RandomRangeInt(Minimum=2, Maximum=3)
    Spawn Group=6
    Use Spawn Probability=false
    UseLoopCountLimit=false
   ) {module=SpawnBurst_Instantaneous006}
   &SpawnBurst_Instantaneous(
    Spawn Count=&RandomRangeInt(Minimum=2, Maximum=3)
    Spawn Group=7
    Use Spawn Probability=false
    UseLoopCountLimit=false
   ) {module=SpawnBurst_Instantaneous007}
  ]
  particleSpawnScript [
   /SetVariables(Particles.VisibilityTag=&SpawnGroup()) {isSetParameters=true}
   /SetVariables(Particles.Random Range:NiagaraFloat=&RandomRangeFloat(Minimum=0.2, Maximum=0.8)) {isSetParameters=true}
   &InitializeParticle(
    Color Mode=::NewEnumerator0
    Lifetime=&MergeHelper_FloatSwitch(
     0=0.4
     1=0.27
     2=0.35
     3=0.35
     4=0.37
     5=0.5
     6=&RandomRangeFloat(Minimum=0.55, Maximum=0.55)
     7=&RandomRangeFloat(Minimum=0.5, Maximum=0.62)
     8=&RandomRangeFloat(Minimum=0.12, Maximum=0.18)
     9=&RandomRangeFloat(Minimum=0.12, Maximum=0.18)
     Index=&SpawnGroup()
     Num=::NewEnumerator8
    )
    UsePositionOffset=false
    Write Material Random=true
   ) {module=InitializeParticle001}
   &CameraOffset(Camera Offset Amount=&MergeHelper_FloatSwitch(0=0, 1=2, 2=2, 3=2, 4=2, 5=2, 6=2, 7=2, 8=10, 9=10, Index=0, Num=::NewEnumerator8))
   &InitialMeshOrientation(
    Orientation Axis=&MergeHelper_VectorSwitch(
     0=<0 0 0>
     1=<0 0 0>
     2=<0 0 0>
     3=<0 0 1>
     4=<0 0 0>
     5=<0 0 0>
     6=<0 0 0>
     7=<0 0 0>
     8=<0 0 0>
     9=<0 0 0>
     Index=&SpawnGroup()
     Num=::NewEnumerator8
    )
    Orientation Vector=&MergeHelper_VectorSwitch(
     0=<0 0 0>
     1=<0 0 0>
     2=<0 0 0>
     3=<0 0 1>
     4=<0 0 0>
     5=<0 0 0>
     6=<0 0 0>
     7=<0 0 0>
     8=<0 0 0>
     9=<0 0 0>
     Index=&SpawnGroup()
     Num=::NewEnumerator8
    )
    Rotation=&MergeHelper_VectorSwitch(
     0=<0 0 0>
     1=<0 0 0>
     2=<0 0 0>
     3=<0 0.25 0>
     4=<0 0 0>
     5=<0 0 0>
     6=<0 0 0>
     7=<0 0 0>
     8=<0 0 0>
     9=<0 0 0>
     Index=&SpawnGroup()
     Num=::NewEnumerator8
    )
    Use Orientation Vector=true
    Use Rotation Vector=true
   )
   &InitialMeshOrientation(
    Orientation Axis=<0 0 0>
    Orientation Vector=<0 0 0>
    Rotation=&MergeHelper_VectorSwitch(
     0=<0 0 0>
     1=<0 0 0>
     2=<0 0 0>
     3=&RandomRangeVector(Minimum=<0 0 0>, Maximum=<1 0 0>)
     4=<0 0 0>
     5=<0 0 0>
     6=<0 0 0>
     7=<0 0 0>
     8=<0 0 0>
     9=<0 0 0>
     Index=&SpawnGroup()
     Num=::NewEnumerator8
    )
    Use Orientation Vector=false
    Use Rotation Vector=true
   ) {module=InitialMeshOrientation001}
   /SetVariables(
    Particles.Dark Glow Color:LinearColor=<rgba 0.0429 0.0227 0.0615 5>
    Particles.Dark Glow Size:Vector2f=&Vector2DFromFloat(Value=130)
   ) {isSetParameters=true}
   /SetVariables(
    Particles.Bright Flare Size:Vector2f=&Vector2DFromFloat(Value=165)
    Particles.Bright Flare color:LinearColor=<rgba 0.4915 0.1564 0.6948 0.5>
   ) {isSetParameters=true}
   /SetVariables(Particles.Tech Disc Size:Vector3f=&VectorFromFloat(Value=1.2386)) {isSetParameters=true}
   /SetVariables(Particles.Shards Disc Size:Vector3f=&VectorFromFloat(Value=0.7)) {isSetParameters=true}
   /SetVariables(
    Particles.Rainbow Color:LinearColor=<rgba 0.6284 0.3343 1 0.06>
    Particles.Rainbow Size:Vector2f=&Vector2DFromFloat(Value=135)
   ) {isSetParameters=true}
   /SetVariables(
    Particles.Darker Glow Color:LinearColor=<rgba 0.0009 0.0058 0.0469 1>
    Particles.Darker Glow Size:Vector2f=&Vector2DFromFloat(Value=215)
   ) {isSetParameters=true}
   /SetVariables(
    Particles.Sharp Sparkles Color:LinearColor=&RandomRangeLinearColor(Minimum=<rgba 1.7297 0.5521 2 1>, Maximum=<rgba 1.6293 0.014 2 1>)
    Particles.Sharp Sparkles Size:Vector2f=&RandomRangeVector2D(Minimum=<0.03 0.005>, Maximum=<0.04 0.02>)
    Particles.Soft Sparkles Size:Vector2f=&Vector2DFromFloat(Value=&RandomRangeFloat(Minimum=15, Maximum=20))
   ) {isSetParameters=true}
   /SetVariables(
    Particles.Long Glitch Dynam 1:Vector4f=&MakeVector4(X=0, Y=0, Z=0, W=&RandomRangeFloat(Minimum=-1, Maximum=0.1))
    Particles.Long Glitch Dynam 2:Vector4f=&MakeVector4(X=&RandomRangeFloat(Minimum=-100, Maximum=100), Y=0, Z=0, W=0)
    Particles.Long Glitch Size:Vector3f=&RandomRangeVector(Minimum=<0.25 2.5 0.25>, Maximum=<0.5 5 1>)
   ) {isSetParameters=true}
   /SetVariables(
    Particles.Square Glitch Dynam 1:Vector4f=&MakeVector4(X=0, Y=0, Z=0, W=&RandomRangeFloat(Minimum=-1, Maximum=0.1))
    Particles.Square Glitch Dynam 2:Vector4f=&MakeVector4(X=&RandomRangeFloat(Minimum=-100, Maximum=100), Y=0, Z=0, W=0)
    Particles.Square Glitch Size:Vector3f=&RandomRangeVector(Minimum=<1 1 2>, Maximum=<2.5 2.5 3>)
   ) {isSetParameters=true}
   &ShapeLocation(
    MaskSpawnGroup=false
    Non Uniform Scale=<1.2 1.2 1>
    Override Seed=false
    Sphere Radius=&MergeHelper_FloatSwitch(0=0, 1=0, 2=0, 3=0, 4=0, 5=0, 6=3, 7=3, 8=20, 9=20, Index=&SpawnGroup(), Num=::NewEnumerator8)
    UseOwnerScale=true
   )
   &AddVelocity(
    Velocity=&MergeHelper_VectorSwitch(
     0=<0 0 0>
     1=<0 0 0>
     2=<0 0 0>
     3=<0 0 0>
     4=<0 0 0>
     5=<0 0 0>
     6=$Particles.ShapeLocation.ShapeNormal
     7=$Particles.ShapeLocation.ShapeNormal
     8=<0 0 0>
     9=<0 0 0>
     Index=&SpawnGroup()
     Num=::NewEnumerator8
    )
    Velocity Speed Scale=&RandomRangeFloat(Minimum=120, Maximum=220)
   )
   &ApplyInitialForces()
  ]
  particleUpdateScript [
   &ParticleState()
   &UpdateMeshOrientation(
    Rotation Rate=&MaskFloat(Mask Source=0, Mask Parameter=?, Mask=2, Masked Float=0.25, Passthrough Float=0)
    Rotation Vector=&MergeHelper_VectorSwitch(
     0=<0 0 0>
     1=<0 0 0>
     2=<0 0 0>
     3=&RandomRangeVector(Minimum=<0 0 -1>, Maximum=<0 0 1>)
     4=<0 0 0>
     5=<0 0 0>
     6=<0 0 0>
     7=<0 0 0>
     8=<0 0 0>
     9=<0 0 0>
     Index=&SpawnGroup()
     Num=::NewEnumerator8
    )
   )
   /SetVariables(
    Particles.Scale Dark Glow Size:Vector2f=&Multiply_Vector2D(
     A=$Particles.Initial.Dark Glow Size
     B=&Vector2DFromFloat(Value=&FloatFromCurve(FloatCurve=DataInterface() {propertyValues {Curve {defaultValue=1, keys [{tangentMode=RCTM_User, time=0, value=0}, <at 0.0753 0.794>, {tangentMode=RCTM_User, time=1, value=1}]}}}))
    )
    Particles.Update Dark Glow Color:LinearColor=&MakeLinearColorFromVectorAndFloat(
     "Vector (RGB)"=&MakeVectorFromLinearColorRGB(Color=$Particles.Initial.Dark Glow Color)
     "Float (Alpha)"=&FloatFromCurve(FloatCurve=DataInterface() {propertyValues {Curve {
         defaultValue=1
         keys [
          {arriveTangent=4.4319, interpMode=RCIM_Cubic, leaveTangent=4.4319, tangentMode=RCTM_User, time=0, value=0}
          {arriveTangent=0.1982, interpMode=RCIM_Cubic, leaveTangent=0.1982, tangentMode=RCTM_User, time=0.1633, value=1}
          {tangentMode=RCTM_User, time=1, value=0}
         ]
        }}})
    )
    Particles.Update Darker Glow Color:LinearColor=&MakeLinearColorFromVectorAndFloat(
     "Vector (RGB)"=&MakeVectorFromLinearColorRGB(Color=$Particles.Initial.Darker Glow Color)
     "Float (Alpha)"=&FloatFromCurve(FloatCurve=DataInterface() {propertyValues {Curve {
         defaultValue=1
         keys [
          {arriveTangent=4.4319, interpMode=RCIM_Cubic, leaveTangent=4.4319, tangentMode=RCTM_User, time=0, value=0}
          {arriveTangent=0.1982, interpMode=RCIM_Cubic, leaveTangent=0.1982, tangentMode=RCTM_User, time=0.1633, value=1}
          {tangentMode=RCTM_User, time=1, value=0}
         ]
        }}})
    )
   ) {isSetParameters=true}
   /SetVariables(
    Particles.Scale Bright Flare Size:Vector2f=&Multiply_Vector2D(
     A=$Particles.Initial.Bright Flare Size
     B=&Vector2DFromFloat(Value=&FloatFromCurve(FloatCurve=DataInterface() {propertyValues {Curve {keys [
           <at 0 -0.0447>
           <at 0.0141 1.0065>
           <at 0.0287 -0.0508>
           {arriveTangent=-0.3125, interpMode=RCIM_Cubic, leaveTangent=-0.3125, tangentMode=RCTM_User, time=0.0532, value=1}
           {arriveTangent=-3.7966, interpMode=RCIM_Cubic, leaveTangent=-3.7966, tangentMode=RCTM_User, time=1, value=0}
          ]}}}))
    )
    Particles.Update Bright Flare Color:LinearColor=&MakeLinearColorFromVectorAndFloat(
     "Vector (RGB)"=&MakeVectorFromLinearColorRGB(Color=$Particles.Initial.Bright Flare color)
     "Float (Alpha)"=&FloatFromCurve(FloatCurve=DataInterface() {propertyValues {Curve {keys [
          {arriveTangent=-2.225, interpMode=RCIM_Cubic, leaveTangent=-2.225, tangentMode=RCTM_User, time=0.1, value=1}
          {interpMode=RCIM_Cubic, tangentMode=RCTM_User, time=1, value=0}
         ]}}})
    )
   ) {isSetParameters=true}
   /SetVariables(
    Particles.Scale Tech Disc Size:Vector3f=&Multiply_Vector(
     A=$Particles.Initial.Tech Disc Size
     B=&VectorFromFloat(Value=&FloatFromCurve(FloatCurve=DataInterface() {propertyValues {Curve {keys [
           <at 0 0>
           <at 0.0221 1.1639>
           <at 0.0397 1.1705>
           <at 0.0664 0>
           <at 0.1121 0>
           {interpMode=RCIM_Cubic, tangentMode=RCTM_User, time=0.1257, value=0.5354}
           {arriveTangent=-0.6347, interpMode=RCIM_Cubic, leaveTangent=-0.6347, time=0.7521, value=1.36}
           <at 0.9693 0>
          ]}}}))
    )
    Particles.Tech Disc Dynamic Param:Vector4f=&MakeVector4(
     X=&FloatFromCurve(FloatCurve=DataInterface() {propertyValues {Curve {keys [
          {arriveTangent=-1.1295, interpMode=RCIM_Cubic, leaveTangent=-1.1295, tangentMode=RCTM_User, time=0, value=0.5}
          {arriveTangent=-1.0726, tangentMode=RCTM_User, time=1, value=0.2}
         ]}}})
     Y=&FloatFromCurve(FloatCurve=DataInterface() {propertyValues {Curve {keys [
          {arriveTangent=-0.0329, interpMode=RCIM_Cubic, leaveTangent=-0.0329, tangentMode=RCTM_User, time=0, value=1}
          {arriveTangent=-0.0114, tangentMode=RCTM_User, time=1, value=0.9}
         ]}}})
     Z=&FloatFromCurve(FloatCurve=DataInterface() {propertyValues {Curve {keys [<at 0 0>,<at 0.3646 1>,<at 1 1.2>]}}})
     W=0
    )
    Particles.Update Tech Disc Color:LinearColor=&ColorFromCurve(ColorCurve=DataInterface() {propertyValues {
       AlphaCurve {keys [<at 0 2>,<at 0.3089 1>,<at 0.6318 0.2>,<at 0.974 0 RCIM_Cubic>]}
       BlueCurve {keys [<at 0.0043 2>,<at 0.2051 1>,<at 0.4489 1>,<at 0.6634 0.776>]}
       GreenCurve {keys [<at 0.0043 1.6439>,<at 0.2051 0.0002>,<at 0.4489 0.5807>,<at 0.6634 0>]}
       RedCurve {keys [<at 0.0043 0>,<at 0.2051 0.4498>,<at 0.4489 0.2066>,<at 0.6634 0.1952>]}
      }})
   ) {isSetParameters=true}
   /SetVariables(
    Particles.Scale Shards Disc Size:Vector3f=&Multiply_Vector(
     A=$Particles.Initial.Shards Disc Size
     B=&VectorFromFloat(Value=&FloatFromCurve(FloatCurve=DataInterface() {propertyValues {Curve {keys [
           {arriveTangent=4.8028, interpMode=RCIM_Cubic, leaveTangent=4.8028, tangentMode=RCTM_User, time=0, value=0.5}
           {arriveTangent=-4.6563, interpMode=RCIM_Cubic, tangentMode=RCTM_Break, time=0.15, value=1}
           <at 1 1.2>
          ]}}}))
    )
    Particles.Update Shards Disc Color:LinearColor=&ColorFromCurve(ColorCurve=DataInterface() {propertyValues {
       AlphaCurve {keys [<at 0 3>,<at 0.3763 1>,<at 0.5545 0.5>,<at 0.672 0.2>,<at 0.9647 0>]}
       BlueCurve {keys [<at 0 1>,<at 0.2177 1>,<at 0.3334 1>,<at 0.4608 1>,<at 0.5706 1>,<at 0.6951 1>,<at 0.8753 1>]}
       GreenCurve {keys [<at 0 0.3862>,<at 0.2177 0.2264>,<at 0.3334 0.2038>,<at 0.4608 0.4468>,<at 0.5706 0.2038>,<at 0.6951 0.1>,<at 0.8753 0.3862>]}
       RedCurve {keys [<at 0 0.0028>,<at 0.2177 0.6419>,<at 0.3334 0.7557>,<at 0.4608 0.8389>,<at 0.5706 0.7557>,<at 0.6951 0.4221>,<at 0.8753 0.0028>]}
      }})
   ) {isSetParameters=true}
   /SetVariables(
    Particles.Scale Rainbow Size:Vector2f=&Multiply_Vector2D(
     A=$Particles.Initial.Rainbow Size
     B=&Vector2DFromFloat(Value=&FloatFromCurve(FloatCurve=DataInterface() {propertyValues {Curve {keys [
           {arriveTangent=10.7004, interpMode=RCIM_Cubic, leaveTangent=10.7004, tangentMode=RCTM_User, time=0, value=0}
           {arriveTangent=0.4505, interpMode=RCIM_Cubic, leaveTangent=0.4505, tangentMode=RCTM_User, time=0.1298, value=0.7288}
           {arriveTangent=0.1276, interpMode=RCIM_Cubic, leaveTangent=0.1276, tangentMode=RCTM_User, time=1, value=1}
          ]}}}))
    )
    Particles.Update Rainbow Color:LinearColor=&MakeLinearColorFromVectorAndFloat(
     "Vector (RGB)"=&MakeVectorFromLinearColorRGB(Color=$Particles.Initial.Rainbow Color)
     "Float (Alpha)"=&FloatFromCurve(
      FloatCurve=DataInterface() {propertyValues {Curve {keys [
          {interpMode=RCIM_Cubic, tangentMode=RCTM_User, time=0, value=0}
          {interpMode=RCIM_Cubic, tangentMode=RCTM_User, time=0.1762, value=1}
          {interpMode=RCIM_Cubic, tangentMode=RCTM_User, time=1, value=0}
         ]}}}
      Scale Curve=0.15
     )
    )
   ) {isSetParameters=true}
   /SetVariables(
    Particles.Scale Sharp Sparkles Size:Vector2f=&Multiply_Vector2D(
     A=$Particles.Initial.Rainbow Size
     B=&Vector2DFromFloat(Value=&FloatFromCurve(
       FloatCurve=DataInterface() {propertyValues {Curve {keys [
           {arriveTangent=10.7004, interpMode=RCIM_Cubic, leaveTangent=10.7004, tangentMode=RCTM_User, time=0, value=0}
           {arriveTangent=0.4505, interpMode=RCIM_Cubic, leaveTangent=0.4505, tangentMode=RCTM_User, time=0.1298, value=0.7288}
           {arriveTangent=0.1276, interpMode=RCIM_Cubic, leaveTangent=0.1276, tangentMode=RCTM_User, time=1, value=1}
          ]}}}
       Scale Curve=0.3
      ))
    )
    Particles.Scale Soft Sparkles Size:Vector2f=&Multiply_Vector2(
     A=$Particles.Initial.Soft Sparkles Size
     B=&Vector2DFromFloat(Value=&FloatFromCurve(
       FloatCurve=DataInterface() {propertyValues {Curve {keys [
           <at 0 0 RCIM_Cubic>
           {arriveTangent=-0.0175, interpMode=RCIM_Cubic, leaveTangent=-0.0175, tangentMode=RCTM_User, time=0.1563, value=0.907}
           {arriveTangent=-0.0042, interpMode=RCIM_Cubic, leaveTangent=-0.0042, tangentMode=RCTM_User, time=0.2646, value=0.0384}
           {arriveTangent=0.0081, interpMode=RCIM_Cubic, leaveTangent=0.0081, tangentMode=RCTM_User, time=0.4692, value=0.7123}
           {arriveTangent=0.034, interpMode=RCIM_Cubic, leaveTangent=0.034, tangentMode=RCTM_User, time=0.6406, value=0.0348}
           {arriveTangent=0.019, interpMode=RCIM_Cubic, leaveTangent=0.019, tangentMode=RCTM_User, time=0.7993, value=0.5053}
           <at 1 0 RCIM_Cubic>
          ]}}}
       CurveIndex=&Multiply_Float(A=$Particles.NormalizedAge, B=$Particles.Random Range)
       Scale Curve=&RandomRangeFloat(Minimum=1, Maximum=3)
      ))
    )
    Particles.Update Sharp Sparkles Color:LinearColor=&MakeLinearColorFromVectorAndFloat(
     "Vector (RGB)"=&MakeVectorFromLinearColorRGB(Color=$Particles.Initial.Sharp Sparkles Color)
     "Float (Alpha)"=&FloatFromCurve(FloatCurve=DataInterface() {propertyValues {Curve {keys [<at 0 0>,<at 0.1 1>,<at 0.5 1>,<at 1 0>]}}}, Scale Curve=0.3)
    )
    Particles.Update Soft Sparkles Color:LinearColor=&ColorFromCurve(ColorCurve=DataInterface() {propertyValues {
       AlphaCurve {keys [<at 0 0>,<at 0.15 0.15 RCIM_Cubic>,<at 1 0 RCIM_Cubic>]}
       BlueCurve {keys [<at 0.1483 0.8796 RCIM_Cubic>,<at 1 2 RCIM_Cubic>]}
       GreenCurve {keys [<at 0.1483 0.9767 RCIM_Cubic>,<at 1 0.4795 RCIM_Cubic>]}
       RedCurve {keys [<at 0.1483 1 RCIM_Cubic>,<at 1 1.2015 RCIM_Cubic>]}
      }})
   ) {isSetParameters=true}
   /SetVariables(
    Particles.Long Glitch Dynam 1:Vector4f=&MakeVector4(
     X=0
     Y=&RandomRangeFloat(Minimum=-1, Maximum=1)
     Z=&FloatFromCurve(FloatCurve=DataInterface() {propertyValues {Curve {keys [
          <at 0 2 RCIM_Cubic>
          {arriveTangent=-6.2285, interpMode=RCIM_Cubic, leaveTangent=-6.2285, tangentMode=RCTM_User, time=0.5612, value=1.0841}
          {arriveTangent=-0.8021, interpMode=RCIM_Cubic, leaveTangent=-0.8021, tangentMode=RCTM_User, time=1, value=0}
         ]}}})
     W=0
    )
    Particles.Long Glitch Dynam 2:Vector4f=&MakeVector4(
     X=0
     Y=0
     Z=0
     W=&FloatFromCurve(FloatCurve=DataInterface() {propertyValues {Curve {keys [
          {arriveTangent=4.1785, interpMode=RCIM_Cubic, leaveTangent=4.1785, tangentMode=RCTM_User, time=0, value=0}
          {arriveTangent=0.9995, interpMode=RCIM_Cubic, leaveTangent=0.9995, tangentMode=RCTM_User, time=0.2223, value=0.6825}
          {arriveTangent=-0.0405, interpMode=RCIM_Cubic, leaveTangent=-0.0405, tangentMode=RCTM_User, time=1, value=1}
         ]}}})
    )
    Particles.Scale Long Glitch Size:Vector3f=&Multiply_Vector(
     A=$Particles.Initial.Long Glitch Size
     B=&Multiply_VectorByFloat(
      Vector=<1 1 1>
      Float=&Multiply_Float(
       A=0.08
       B=&FloatFromCurve(
        FloatCurve=DataInterface() {propertyValues {Curve {keys [
            {arriveTangent=5.4245, interpMode=RCIM_Cubic, leaveTangent=5.4245, tangentMode=RCTM_User, time=0, value=5}
            {arriveTangent=-3.9614, interpMode=RCIM_Cubic, leaveTangent=-3.9614, time=0.0963, value=8.5355}
            {arriveTangent=-22.2662, interpMode=RCIM_Cubic, leaveTangent=-22.2662, time=0.2702, value=3.9295}
            <at 0.3 4>
            <at 0.4 11.5>
            {arriveTangent=0.1532, interpMode=RCIM_Cubic, leaveTangent=0.1532, tangentMode=RCTM_User, time=0.5, value=2.875}
            <at 0.6 3>
            <at 0.7 11.5>
            <at 0.8 2.5>
            <at 0.9 2>
            {arriveTangent=-0.0405, interpMode=RCIM_Cubic, leaveTangent=-0.0405, tangentMode=RCTM_User, time=1, value=0.5}
           ]}}}
        Scale Curve=&Multiply_Float(
         A=0.6
         B=&ClampFloat(
          Float=&FN_LOD_Distance_Remap(Start Distance=200, End Distance=400, Minimum Output=0.5, Maximum Output=1)
          Min=0.5
         )
        )
       )
      )
     )
    )
    Particles.Update Long Glitch Color:LinearColor=&RandomRangeLinearColor(Minimum=<rgba 0.0769 0.3888 1 1>, Maximum=<rgba 0.3403 0.8457 1 1>)
   ) {isSetParameters=true}
   /SetVariables(
    Particles.Scale Square Glitch Size:Vector3f=&Multiply_Vector(
     A=$Particles.Initial.Square Glitch Size
     B=&Multiply_VectorByFloat(
      Vector=<1 1 1>
      Float=&Multiply_Float(
       A=0.08
       B=&FloatFromCurve(
        FloatCurve=DataInterface() {propertyValues {Curve {keys [
            {arriveTangent=5.4245, interpMode=RCIM_Cubic, leaveTangent=5.4245, tangentMode=RCTM_User, time=0, value=5}
            {arriveTangent=-3.9614, interpMode=RCIM_Cubic, leaveTangent=-3.9614, time=0.0963, value=8.5355}
            {arriveTangent=-22.2662, interpMode=RCIM_Cubic, leaveTangent=-22.2662, time=0.2702, value=3.9295}
            <at 0.3 4>
            <at 0.4 11.5>
            {arriveTangent=0.1532, interpMode=RCIM_Cubic, leaveTangent=0.1532, tangentMode=RCTM_User, time=0.5, value=2.875}
            <at 0.6 3>
            <at 0.7 11.5>
            <at 0.8 2.5>
            <at 0.9 2>
            {arriveTangent=-0.0405, interpMode=RCIM_Cubic, leaveTangent=-0.0405, tangentMode=RCTM_User, time=1, value=0.5}
           ]}}}
        Scale Curve=&Multiply_Float(
         A=0.6
         B=&ClampFloat(
          Float=&FN_LOD_Distance_Remap(Start Distance=200, End Distance=400, Minimum Output=0.5, Maximum Output=1)
          Min=0.5
         )
        )
       )
      )
     )
    )
    Particles.Square Glitch Dynam 1:Vector4f=&MakeVector4(
     X=0
     Y=&RandomRangeFloat(Minimum=-1, Maximum=1)
     Z=&FloatFromCurve(FloatCurve=DataInterface() {propertyValues {Curve {keys [
          <at 0 2 RCIM_Cubic>
          {arriveTangent=-6.2285, interpMode=RCIM_Cubic, leaveTangent=-6.2285, tangentMode=RCTM_User, time=0.5612, value=1.0841}
          {arriveTangent=-0.8021, interpMode=RCIM_Cubic, leaveTangent=-0.8021, tangentMode=RCTM_User, time=1, value=0}
         ]}}})
     W=0
    )
    Particles.Square Glitch Dynam 2:Vector4f=&MakeVector4(
     X=&RandomRangeFloat(Minimum=-1, Maximum=1)
     Y=0
     Z=0
     W=&FloatFromCurve(FloatCurve=DataInterface() {propertyValues {Curve {keys [
          {arriveTangent=4.1785, interpMode=RCIM_Cubic, leaveTangent=4.1785, tangentMode=RCTM_User, time=0, value=0}
          {arriveTangent=0.9995, interpMode=RCIM_Cubic, leaveTangent=0.9995, tangentMode=RCTM_User, time=0.2223, value=0.6825}
          {arriveTangent=-0.0405, interpMode=RCIM_Cubic, leaveTangent=-0.0405, tangentMode=RCTM_User, time=1, value=1}
         ]}}})
    )
    Particles.Update Square Glitch Color:LinearColor=&RandomRangeLinearColor(Minimum=<rgba 0.2594 0.0324 1 1>, Maximum=<rgba 0.4083 0.0714 1 1>)
   ) {isSetParameters=true}
   &Drag(
    Drag=&MergeHelper_FloatSwitch(
     0=0
     1=0
     2=0
     3=0
     4=0
     5=0
     6=&FloatFromCurve(
      FloatCurve=DataInterface() {propertyValues {Curve {keys [
          {arriveTangent=2.0771, interpMode=RCIM_Cubic, leaveTangent=2.0771, tangentMode=RCTM_User, time=0, value=0}
          {interpMode=RCIM_Cubic, tangentMode=RCTM_User, time=1, value=1}
         ]}}}
      Scale Curve=10
     )
     7=&FloatFromCurve(
      FloatCurve=DataInterface() {propertyValues {Curve {keys [
          {arriveTangent=2.0771, interpMode=RCIM_Cubic, leaveTangent=2.0771, tangentMode=RCTM_User, time=0, value=0}
          {interpMode=RCIM_Cubic, tangentMode=RCTM_User, time=1, value=1}
         ]}}}
      Scale Curve=10
     )
     Index=&SpawnGroup()
     Num=::NewEnumerator6
    )
    Rotational Drag:NiagaraFloat=&MergeHelper_FloatSwitch(
     0=0
     1=0
     2=0
     3=0
     4=0
     5=0
     6=&FloatFromCurve(
      FloatCurve=DataInterface() {propertyValues {Curve {keys [
          {arriveTangent=2.0771, interpMode=RCIM_Cubic, leaveTangent=2.0771, tangentMode=RCTM_User, time=0, value=0}
          {interpMode=RCIM_Cubic, tangentMode=RCTM_User, time=1, value=1}
         ]}}}
      Scale Curve=3
     )
     7=&FloatFromCurve(
      FloatCurve=DataInterface() {propertyValues {Curve {keys [
          {arriveTangent=2.0771, interpMode=RCIM_Cubic, leaveTangent=2.0771, tangentMode=RCTM_User, time=0, value=0}
          {interpMode=RCIM_Cubic, tangentMode=RCTM_User, time=1, value=1}
         ]}}}
      Scale Curve=3
     )
     Index=&SpawnGroup()
     Num=::NewEnumerator6
    )
    Use Linear Drag=true
    Use Rotational Drag=true
   )
   &SolveForcesAndVelocity(Clamp Velocity=false, Limit Acceleration=false)
  ]
  rendererClasses [&NiagaraSpriteRendererProperties, &NiagaraMeshRendererProperties]
  renderers [
   {rendererClass=&NiagaraSpriteRendererProperties}
   {rendererClass=&NiagaraSpriteRendererProperties, rendererIndex=1}
   {rendererClass=&NiagaraMeshRendererProperties, rendererIndex=2}
   {rendererClass=&NiagaraSpriteRendererProperties, rendererIndex=3}
   {rendererClass=&NiagaraSpriteRendererProperties, rendererIndex=4}
   {rendererClass=&NiagaraSpriteRendererProperties, rendererIndex=5}
   {rendererClass=&NiagaraSpriteRendererProperties, rendererIndex=6}
   {rendererClass=&NiagaraMeshRendererProperties, rendererIndex=7}
   {rendererClass=&NiagaraMeshRendererProperties, rendererIndex=8}
  ]
 }
}
summary {
 systemName=NS_Emote_MincePounce_Cat_Happy
 emitters [
  {emitterName=Cat, rendererClasses []}
  {emitterName=Light, rendererClasses []}
  {emitterName=End_Burst, rendererClasses [&NiagaraSpriteRendererProperties]}
  {emitterName=Spawn_Burst, rendererClasses [&NiagaraSpriteRendererProperties, &NiagaraMeshRendererProperties]}
 ]
 userVariables [
  User.bIsFrontEnd {type=NiagaraBool}
  User.bIsFrontEndPreview {type=NiagaraBool}
  User.FrontEndScale {type=NiagaraFloat, defaultValue=0.35, description=""}
  User.WipeDuration {type=NiagaraFloat, defaultValue=0.8, description=""}
 ]
}
